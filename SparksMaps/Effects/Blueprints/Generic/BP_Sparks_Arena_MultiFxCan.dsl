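; Arena stage FX can that fires the current Niagara system on stage show Activate (strength above 0.5) and Impulse events.
; On cosmetic-capable instances, BeginPlay seeds a component pool from StageFX_01 up to the initial size.
; Each trigger takes an inactive component, grows the pool up to a maximum, or logs and drops the event.
; Stage show events and Mojo functions duplicate this logic; other Mojo handlers are stubs.
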(event EventBeginPlay
  (Networking|CanExecuteCosmeticEvents
    (:True
      (Utilities|Array|Add (Variables|Default|GetNiagaraParticleSystems) (Variables|Default|GetStageFX_01))
      (if (> (Variables|Default|GetInitialPoolSize) 1)
        (for _ (range 1 (Variables|Default|GetInitialPoolSize))
          (bind _returnvalue (Game|AddComponentbyClass 0 "/Script/Niagara.NiagaraComponent" false (Transformation|GetRelativeTransform (Variables|Default|GetStageFX_01))))
          (Utilities|Array|Add (Variables|Default|GetNiagaraParticleSystems) _returnvalue))
        (Development|PrintText (Utilities|Text|FormatText "NSLOCTEXT(\"[9E490ED2D38775D0EA95329E6EDA4C18]\", \"CD643B404D016BBAA95640B76355B099\", \"[FxCan] {name} increased pool size: {size}\")" (Utilities|GetDisplayName self) (Utilities|Array|Length (Variables|Default|GetNiagaraParticleSystems))) false true "(R=1.000000,G=0.450514,B=0.000000,A=1.000000)" 1e+01 (Utilities|String|StringToName (Utilities|GetDisplayName self)))))))

(event EventStageShowUpdateNiagaraSystem
  (params (NiagaraSystem :type Niagara%20System%20Object%20Reference))
  ;[1 1 1] This function needs to be implemented to support setting Niagara Systems through Stage Events
  (Variables|Default|SetCurrentNiagaraSystem NiagaraSystem))

(event EventStageShowActivate
  (params (Strength :type Float%20%28single-precision%29) (MusicTimeMeasure :type Float%20%28single-precision%29))
  (if (> Strength 0.5)
    (bind _npscomponent (CallFunction|GetNSComponentfromPool))
    (Utilities|IsValid _npscomponent
      (:Is%20Valid
        (Niagara|SetNiagaraSystemAsset _npscomponent (Variables|Default|GetCurrentNiagaraSystem) true)
        (Components|Activation|Activate _npscomponent true)))))

(event EventStageShowImpulse
  (params (MusicTimeMeasure :type Float%20%28single-precision%29))
  (bind _npscomponent (CallFunction|GetNSComponentfromPool))
  (Utilities|IsValid _npscomponent
    (:Is%20Valid
      (Niagara|SetNiagaraSystemAsset _npscomponent (Variables|Default|GetCurrentNiagaraSystem) true)
      (Components|Activation|Activate _npscomponent true))
    (:Is%20Not%20Valid
      (Development|PrintText (Utilities|Text|FormatText "NSLOCTEXT(\"[9E490ED2D38775D0EA95329E6EDA4C18]\", \"163F71F041655EC5B266EC90E5014C3F\", \"[FxCan] {name} ignored event (no NPS component available)\")" (Utilities|GetDisplayName self)) false true "(R=1.000000,G=0.317046,B=0.052083,A=1.000000)" 1e+01 (Utilities|String|StringToName (Utilities|GetDisplayName self))))))

(fn GetNSComponentFromPool
  ;[1 1 1] Tries to find an inactive Niagara Particle System Component that's inactive on the pool
  (bind (_array_element _array_index) (Utilities|Array|ForEachLoopwithBreak (Variables|Default|GetNiagaraParticleSystems))
    (:LoopBody
      (if (not (Components|Activation|IsActive _array_element))
        (Variables|Default|SetPsIndex _array_index)
        (LocalVariables|SetFoundFreeNPSComponent true)))
    (:Completed
      _))
  (if (LocalVariables|GetFoundFreeNPSComponent)
    _
    (elif (< (Utilities|Array|Length (Variables|Default|GetNiagaraParticleSystems)) (Variables|Default|GetMaxPoolSize))
      ;[1 1 1] Add new NPS Component, increasing the pool
      (bind _returnvalue (Game|AddComponentbyClass 0 "/Script/Niagara.NiagaraComponent" false (Transformation|GetRelativeTransform (Variables|Default|GetStageFX_01))))
      (bind _returnvalue_2 (Utilities|Array|Add (Variables|Default|GetNiagaraParticleSystems) _returnvalue))
      (Variables|Default|SetPsIndex _returnvalue_2)
      (Development|PrintText (Utilities|Text|FormatText "NSLOCTEXT(\"[9E490ED2D38775D0EA95329E6EDA4C18]\", \"A110631D495E88E86B1A5C9E0A921B2C\", \"[FxCan] {name} increased pool size: {size}\")" (Utilities|GetDisplayName self) (Utilities|Array|Length (Variables|Default|GetNiagaraParticleSystems))) false true "(R=1.000000,G=0.450514,B=0.000000,A=1.000000)" 1e+01 (Utilities|String|StringToName (Utilities|GetDisplayName self)))
      (else
        ;[1 0.3902 0.3304] Return null if we can't increase pool size
        (Development|PrintText (Utilities|Text|FormatText "NSLOCTEXT(\"[9E490ED2D38775D0EA95329E6EDA4C18]\", \"F8749B6C45967FD71B7089BCCE8D813A\", \"[FxCan] {name} ignored event (no NPS component available)\")" (Utilities|GetDisplayName self)) false true "(R=1.000000,G=0.317046,B=0.052083,A=1.000000)" 1e+01 (Utilities|String|StringToName (Utilities|GetDisplayName self)))
        (return 0))))
  (return (Utilities|Array|Get(acopy) (Variables|Default|GetNiagaraParticleSystems) (Variables|Default|GetPsIndex))))

(fn MojoStageShowActivate
  (params (Strength :type Float%20%28double-precision%29) (MusicTimeMeasure :type Float%20%28double-precision%29))
  (if (> Strength 0.5)
    (bind _npscomponent (CallFunction|GetNSComponentfromPool))
    (Utilities|IsValid _npscomponent
      (:Is%20Valid
        (Niagara|SetNiagaraSystemAsset _npscomponent (Variables|Default|GetCurrentNiagaraSystem) true)
        (Components|Activation|Activate _npscomponent false)
        (return false)))))

(fn MojoStageShowActivateBeat
  (params (Active :type Float%20%28double-precision%29) (MusicTimeMeasure :type Float%20%28double-precision%29))
  (return false))

(fn MojoStageShowColorShift
  (params (Primary :type Linear%20Color%20Structure) (Secondary :type Linear%20Color%20Structure) (Tertiary :type Linear%20Color%20Structure) (AltPrimary :type Linear%20Color%20Structure) (AltSecondary :type Linear%20Color%20Structure) (AltTertiary :type Linear%20Color%20Structure) (MusicTimeMeasure :type Float%20%28double-precision%29))
  (return false))

(fn MojoStageShowCustomMode
  (params (ModeIndex :type Integer) (MusicTimeMeasure :type Float%20%28double-precision%29))
  (return false))

(fn MojoStageShowImpulse
  (params (MusicTimeMeasure :type Float%20%28double-precision%29))
  (bind _npscomponent (CallFunction|GetNSComponentfromPool))
  (Utilities|IsValid _npscomponent
    (:Is%20Valid
      (Niagara|SetNiagaraSystemAsset _npscomponent (Variables|Default|GetCurrentNiagaraSystem) true)
      (Components|Activation|Activate _npscomponent false))
    (:Is%20Not%20Valid
      (Development|PrintText (Utilities|Text|FormatText "NSLOCTEXT(\"[9E490ED2D38775D0EA95329E6EDA4C18]\", \"163F71F041655EC5B266EC90E5014C3F\", \"[FxCan] {name} ignored event (no NPS component available)\")" (Utilities|GetDisplayName self)) false true "(R=1.000000,G=0.317046,B=0.052083,A=1.000000)" 1e+01 (Utilities|String|StringToName (Utilities|GetDisplayName self)))))
  (return false))

(fn MojoStageShowOnStop
  (return false))

(fn MojoStageShowUpdateNiagaraSystem
  (params (Niagara%20System :type Niagara%20System%20Object%20Reference))
  ;[1 1 1] This function needs to be implemented to support setting Niagara Systems through Stage Events
  (Variables|Default|SetCurrentNiagaraSystem Niagara%20System)
  (return false))

(fn ConstructionScript)
