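; Multi-laser FX can for Sparks stage shows, driving a Niagara Laser component through stage show events and MojoStageShow functions.
; Activate, color shift and impulse defer to the parent class, though color shift omits the alt colors; custom mode just returns false.
; Stopping activates at zero strength.
; When AllowNiagaraChange is set, parallel event and Mojo handlers swap the Laser's system, restore its active state and refresh static color.
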(event EventStageShowUpdateNiagaraSystem
  (params (NiagaraSystem :type Niagara%20System%20Object%20Reference))
  (if (Variables|Default|GetAllowNiagaraChange)
    ;[1 1 1] This function needs to be implemented to support setting Niagara Systems through Stage Events
    (Variables|Default|SetbFxWasActive (Components|Activation|IsActive (Variables|Default|GetLaser)))
    (Components|Activation|Deactivate (Variables|Default|GetLaser))
    (Niagara|SetNiagaraSystemAsset (Variables|Default|GetLaser) NiagaraSystem false)
    (Components|Activation|SetActive (Variables|Default|GetLaser) (Variables|Default|GetbFxWasActive) false)
    (CallFunction|UpdateNiagaraStaticColor)))

(event EventStageShowOnStop
  (CallFunction|MojoStageShowActivate 0.0))

(fn MojoStageShowActivate
  (params (Strength :type Float%20%28double-precision%29) (MusicTimeMeasure :type Float%20%28double-precision%29))
  (bind _result (|CallParentFunction Strength MusicTimeMeasure))
  (return _result))

(fn MojoStageShowColorShift
  (params (Primary :type Linear%20Color%20Structure) (Secondary :type Linear%20Color%20Structure) (Tertiary :type Linear%20Color%20Structure) (AltPrimary :type Linear%20Color%20Structure) (AltSecondary :type Linear%20Color%20Structure) (AltTertiary :type Linear%20Color%20Structure) (MusicTimeMeasure :type Float%20%28double-precision%29))
  (bind _result (|CallParentFunction :Primary Primary :Secondary Secondary :Tertiary Tertiary :MusicTimeMeasure MusicTimeMeasure))
  (return _result))

(fn MojoStageShowCustomMode
  (params (ModeIndex :type Integer) (MusicTimeMeasure :type Float%20%28double-precision%29))
  (return false))

(fn MojoStageShowImpulse
  (params (MusicTimeMeasure :type Float%20%28double-precision%29))
  (bind _result (|CallParentFunction MusicTimeMeasure))
  (return _result))

(fn MojoStageShowOnStop
  (CallFunction|MojoStageShowActivate 0.0)
  (return false))

(fn MojoStageShowUpdateNiagaraSystem
  (params (Niagara%20System :type Niagara%20System%20Object%20Reference))
  (if (Variables|Default|GetAllowNiagaraChange)
    ;[1 1 1] This function needs to be implemented to support setting Niagara Systems through Stage Events
    (LocalVariables|SetFxWasActive (Components|Activation|IsActive (Variables|Default|GetLaser)))
    (Components|Activation|Deactivate (Variables|Default|GetLaser))
    (Niagara|SetNiagaraSystemAsset (Variables|Default|GetLaser) Niagara%20System false)
    (Components|Activation|SetActive (Variables|Default|GetLaser) (LocalVariables|GetFxWasActive) false)
    (CallFunction|UpdateNiagaraStaticColor)
    (return false)))

(fn ConstructionScript
  (|CallParentFunction))
